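; StormInterface, in the StormBeacon folder of the storm controller, is by its name a Blueprint interface for querying the current storm phase.
; It declares one function, GetStormPhase, which takes no inputs and returns a value shown only as a default of zero.
; The file holds no other logic, so any real phase value must come from implementations elsewhere.
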(fn GetStormPhase
  (return 0))
